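; Sound Library context for cosmetic animation sounds, which plays event audio on a Fortnite player pawn outside frontend cosmetic displays.
; Play casts the context actor to FortPlayerPawn and checks whether it is displayed as a cosmetic object in a frontend context.
; If not, it spawns the event's sound attached to the character mesh with a vertical offset and no named socket, returning the audio component in an array.
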(fn Play
  (params (InEventData :type Sound%20Library%20Context%20Event%20Input%20Structure%20%28by%20ref%29))
  (bind _asfort_player_pawn (Utilities|Casting|CastToFortPlayerPawn (Variables|SoundLibrary|GetActor)))
  (if (not (Reactive|Frontend|IsCosmeticObjectDisplayedinFrontendContext _asfort_player_pawn))
    (bind _returnvalue (Audio|SpawnSoundAttached (Utilities|Struct|BreakSoundLibraryContextEventInput InEventData) (Variables|/Script/Engine.Character|GetMesh _asfort_player_pawn) "None" "0.000000,0.000000,110.000000"))
    (return (Utilities|Array|MakeArray _returnvalue) true)))
