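; Sound Library context Blueprint for the Mimosa cosmetic emote.
; Its Play function takes Sound Library context event input by reference but runs no logic.
; It returns fixed zero and false values, so it plays no sound itself.
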(fn Play
  (params (InEventData :type Sound%20Library%20Context%20Event%20Input%20Structure%20%28by%20ref%29))
  (return 0 false))
